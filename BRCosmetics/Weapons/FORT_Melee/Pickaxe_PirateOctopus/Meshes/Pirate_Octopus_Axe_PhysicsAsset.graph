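# Physics asset for the Pirate Octopus two-handed pickaxe, giving its handle and dangling parts physics bodies for secondary motion.
# Capsules cover the handle, octopus body, head, two two-segment scarf tails and rope, with a small sphere on the root.
# Constraints chain root, handle, body and head with free swing and twist.
# Scarves hang from the head and the rope from the handle, with tight twist limits and wider swing on the rope.
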PhysicsAsset /BRCosmetics/Weapons/FORT_Melee/Pickaxe_PirateOctopus/Meshes/Pirate_Octopus_Axe_PhysicsAsset
bodies {
 Melee_TwoHanded_Handle [{shapeType=Capsule, length=90.5709, radius=7.6811, center=<-0.4423 -0.8407 13.6816>}]
 b_AR_Root [{shapeType=Sphere, radius=0.505}]
 dyn_body [{shapeType=Capsule, length=36.1445, radius=15.1025, center=<-13.2952 -8.673 -2.1828>, rotation=<rot 47.5287 -62.4743 25.0934>}]
 dyn_head [{shapeType=Capsule, length=0, radius=3.9631, center=<9.461 0.2432 0.4231>, rotation=<rot 2.4367 -90 -14.2805>}]
 dyn_l_scarf_01 [{shapeType=Capsule, length=3.3026, radius=1.3821, center=<2.9262 0.0598 -0.7646>, rotation=<rot -6.5104 83.3625 -98.8479>}]
 dyn_l_scarf_02 [{shapeType=Capsule, length=2.4997, radius=1.1276, center=<2.4592 -0.0179 -0.0882>, rotation=<rot 0 -90.4756 -86.5338>}]
 dyn_r_scarf_01 [{shapeType=Capsule, length=4.5293, radius=2.1366, center=<4.2767 -0.0005 1.146>, rotation=<rot -74.9995 -0.0002 0.0002>}]
 dyn_r_scarf_02 [{shapeType=Capsule, length=3.0431, radius=1.7589, center=<3.2799 0.0178 0.0356>, rotation=<rot 0 88.7605 -88.3227>}]
 dyn_rope [{shapeType=Capsule, length=6.5095, radius=2.5, center=<6.0869 0.4302 -0.3383>, rotation=<rot -0.0217 -92.2395 -89.7526>}]
}
constraints [
 {
  bone1Name=Melee_TwoHanded_Handle
  bone2Name=b_AR_Root
  swing1LimitDegrees=45
  swing1Motion=Free
  swing2LimitDegrees=45
  swing2Motion=Free
  twistLimitDegrees=45
  twistMotion=Free
 }
 {
  bone1Name=dyn_body
  bone2Name=Melee_TwoHanded_Handle
  swing1LimitDegrees=45
  swing1Motion=Free
  swing2LimitDegrees=45
  swing2Motion=Free
  twistLimitDegrees=45
  twistMotion=Free
 }
 {bone1Name=dyn_r_scarf_01, bone2Name=dyn_head, swing1LimitDegrees=45, swing2LimitDegrees=45, twistLimitDegrees=10}
 {bone1Name=dyn_r_scarf_02, bone2Name=dyn_r_scarf_01, swing1LimitDegrees=45, swing2LimitDegrees=45, twistLimitDegrees=10}
 {bone1Name=dyn_l_scarf_01, bone2Name=dyn_head, swing1LimitDegrees=45, swing2LimitDegrees=45, twistLimitDegrees=10}
 {bone1Name=dyn_l_scarf_02, bone2Name=dyn_l_scarf_01, swing1LimitDegrees=45, swing2LimitDegrees=45, twistLimitDegrees=10}
 {bone1Name=dyn_rope, bone2Name=Melee_TwoHanded_Handle, swing1LimitDegrees=80, swing2LimitDegrees=80, twistLimitDegrees=10}
 {
  bone1Name=dyn_head
  bone2Name=dyn_body
  swing1LimitDegrees=45
  swing1Motion=Free
  swing2LimitDegrees=45
  swing2Motion=Free
  twistLimitDegrees=45
  twistMotion=Free
 }
]
